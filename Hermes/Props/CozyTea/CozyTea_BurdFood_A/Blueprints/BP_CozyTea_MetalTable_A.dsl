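; Metal table prop Blueprint for the CozyTea set, dressed with randomized meshes through an ISM_SocketFiller component.
; Construction calls the parent, then rebuilds all socket fillers on the actor via SocketFillerFunctionLibrary.
; As an exposed asset provider, it adds its version and reports each randomized filler mesh as an indirect asset reference to allow.
; BeginPlay, overlap and Tick events are empty stubs.
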(event EventBeginPlay)

(event Collision|EventActorBeginOverlap
  (params (OtherActor :type Actor%20Object%20Reference)))

(event EventTick
  (params (DeltaSeconds :type Float%20%28single-precision%29)))

(fn BuildExposedAssetProviderVersion(Implementation)
  (params (Version :type Fort%20Exposed%20Asset%20Provider%20Version%20Structure%20%28by%20ref%29))
  (|CallParentFunction Version)
  (ExposedAssetProvider|AddVersion Version 2)
  (return))

(fn CollectAssetReferencestoAllow(Implementation)
  (params (Collector :type Fort%20Exposed%20Asset%20Reference%20Collector%20Structure%20%28by%20ref%29))
  (|CallParentFunction Collector)
  (bind _returnvalue (Fortnite|Utilities|ResolveComponentFromActor 0 (Variables|Default|GetISM_SocketFiller)))
  (for _array_element (Variables|/Game/Environments/Blueprints/SocketInstancerComponent/ISM_SocketFiller.ISM_SocketFiller_C|GetRandomizedMeshes _returnvalue)
    (ExposedAssetProvider|CollectIndirectAssetReference Collector (Utilities|Struct|BreakSocketFillerMeshStruct _array_element)))
  (return))

(fn ConstructionScript
  (|CallParentFunction)
  (Class|SocketFillerFunctionLibrary|ReconstructAllSocketFillers self))
